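# Looping ambient cue for the Zero Point top emitter, playing a close-range loop in the Athena ambient sound class.
# Two chained continuous modulators shape the wave player from runtime parameters.
# PulseRampUp lowers pitch and ducks volume as it rises, while Progress raises pitch early in its range and boosts volume later.
# Attenuation comes from the shared ZP_Close_Att asset; the cue's stored override values are not switched on.
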SoundCue /Game/Sounds/Apollo/Biomes/ZeroPoint/AMB_Emit_ZP_Top_Loop_Cue
root=ModulatorContinuous_0
nodes [
 {
  id=ModulatorContinuous_0
  children [ModulatorContinuous_1]
  properties {
   pitchModulationParams {default=0, maxOutput=0.7, minOutput=1, parameterName=PulseRampUp}
   volumeModulationParams {default=0, maxOutput=0.1, minOutput=1, parameterName=PulseRampUp}
  }
 }
 {
  id=ModulatorContinuous_1
  children [WavePlayer_3]
  properties {
   pitchModulationParams {default=0, maxInput=0.4, maxOutput=1.2, minOutput=0.9, parameterName=Progress}
   volumeModulationParams {default=0, maxOutput=1.5, minInput=0.4, minOutput=1, parameterName=Progress}
  }
 }
 {id=WavePlayer_3, properties {bLooping=true, soundWaveAssetPtr=/Game/Athena/Sounds/Events/ZP2/ZP_Amb_Close_Loop_01}}
]
properties {
 attenuationSettings=/Game/Athena/Sounds/Events/ZP2/ZP_Close_Att
 maxDistance=5000
 soundClassObject=&Fort_Ambient_Athena
 volumeMultiplier=0.501
 attenuationOverrides {
  bAttenuateWithLPF=true
  bEnableFocusInterpolation=true
  bEnableListenerFocus=true
  distanceAlgorithm=Custom
  falloffDistance=5500
  focusAttackInterpSpeed=0.5
  focusAzimuth=15
  focusDistanceScale=0.5
  focusReleaseInterpSpeed=0.25
  lPFFrequencyAtMax=500
  lPFRadiusMax=2000
  lPFRadiusMin=1000
  nonFocusAzimuth=90
  nonFocusVolumeAttenuation=0.1
  stereoSpread=2048
  attenuationShapeExtents {x=1024}
  customAttenuationCurve {editorCurveData {keys [<at 0 1>,<at 0.3 1>,<at 0.4 0.5>,<at 1 0>]}}
 }
}
