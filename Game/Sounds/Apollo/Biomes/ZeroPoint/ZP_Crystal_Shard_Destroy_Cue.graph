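# Sound cue for a Zero Point crystal shard being destroyed, classed as an impact sound in the Apollo biome audio.
# A short fixed delay feeds a default-setting modulator, then a random node picks one of three equally weighted crystal destruction waves.
# It references the small-object demolition attenuation asset and overrides concurrency so new instances are blocked at the limit instead of stopping playing ones.
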SoundCue /Game/Sounds/Apollo/Biomes/ZeroPoint/ZP_Crystal_Shard_Destroy_Cue
root=Delay_0
nodes [
 {id=Delay_0, children [Modulator_0], properties {delayMax=0.1, delayMin=0.1}}
 {id=Modulator_0, children [Random_1]}
 {id=Random_1, children [WavePlayer_0,WavePlayer_1,WavePlayer_2], properties {weights [1,1,1]}}
 {id=WavePlayer_0, properties {soundWaveAssetPtr=&/ZP_Crystal_Destroyed_03}}
 {id=WavePlayer_1, properties {soundWaveAssetPtr=&/ZP_Crystal_Destroyed_01}}
 {id=WavePlayer_2, properties {soundWaveAssetPtr=&/ZP_Crystal_Destroyed_02}}
]
properties {
 attenuationSettings=/Game/Sounds/Attenuation/Fort_Demolition_Destroy_Small_Sized_Object_Att
 bOverrideConcurrency=true
 maxDistance=4000
 soundClassObject=&Fort_Impacts
 volumeMultiplier=1.5
 attenuationOverrides {falloffDistance=2600}
 concurrencyOverrides {resolutionRule=PreventNew}
}
