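# Sound cue for crystal shard impacts in the Zero Point biome.
# A random node picks one of six equally weighted crystal impact waves, and a modulator on top varies each play.
# It uses the impacts sound class and a small-object demolition hit attenuation.
# A concurrency override prevents new instances once the limit is reached.
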SoundCue /Game/Sounds/Apollo/Biomes/ZeroPoint/ZP_Crystal_Shard_Impact_Cue
root=Modulator_2
nodes [
 {id=Modulator_2, children [Random_0]}
 {
  id=Random_0
  children [WavePlayer_0, WavePlayer_1, WavePlayer_2, WavePlayer_3, WavePlayer_9, WavePlayer_10]
  properties {weights [1,1,1,1,1,1]}
 }
 {id=WavePlayer_0, properties {soundWaveAssetPtr=&/ZP_Crystal_Impt_01}}
 {id=WavePlayer_1, properties {soundWaveAssetPtr=&/ZP_Crystal_Impt_02}}
 {id=WavePlayer_10, properties {soundWaveAssetPtr=&/ZP_Crystal_Impt_06}}
 {id=WavePlayer_2, properties {soundWaveAssetPtr=&/ZP_Crystal_Impt_03}}
 {id=WavePlayer_3, properties {soundWaveAssetPtr=&/ZP_Crystal_Impt_04}}
 {id=WavePlayer_9, properties {soundWaveAssetPtr=&/ZP_Crystal_Impt_05}}
]
properties {
 attenuationSettings=/Game/Sounds/Attenuation/Fort_Demolition_Hit_Small_Sized_Object_Att
 bOverrideConcurrency=true
 maxDistance=4000
 soundClassObject=&Fort_Impacts
 volumeMultiplier=0.84
 attenuationOverrides {falloffDistance=2600}
 concurrencyOverrides {resolutionRule=PreventNew}
}
